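# Library material function that prepares a particle material for a merged "LW" emitter, exposing MLW_ parameters that adjust Color, Opacity and WPO.
# Color and Opacity are multiplied by their scale parameters, with Enable scalars blending toward those parameter values as overrides.
# WPO sums sprite scaling in tangent space, UV-derived corner offsets along tangent and bitangent, and mesh and incoming-WPO scaling in instance space, with size scaling relative to particle position.
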MaterialFunction /Game/Effects/Fort_Effects/Materials/_Functions/MF_PrepareForMergeLWEmitter
graph {
 properties {bExposeToLibrary=true}
 nodes [
  n1 = Subtract(A=WorldPosition() {worldPositionShaderOffset=WPT_ExcludeAllShaderOffsets}.XYZ, B=ParticlePositionWS())
  n2 = Transform(None=@n1) {transformSourceType=TRANSFORMSOURCE_World, transformType=TRANSFORM_Tangent}
  n3 = Subtract(A=Multiply(A=@n2, B=VectorParameter() {parameterName=MLW_SpriteSizeScale, defaultValue=<rgba 1 1 1 1>}.RGB), B=@n2)
  n4 = Ceil(None=Subtract(A=TextureCoordinate()) {constB=0.5})
  n5 = MaterialFunctionCall(Float2=Subtract(A=Multiply(A=@n4) {constB=2})) {materialFunction=&BreakOutFloat2Components}
  n6 = VertexTangentWS()
  n7 = VectorParameter() {parameterName=MLW_SpriteSizeOffset, defaultValue=<rgba 0 0 1 1>}
  n8 = CrossProduct(A=VertexNormalWS(), B=@n6)
  n9 = Add(A=Multiply(A=Multiply(A=@n5.R, B=@n6), B=@n7.R), B=Multiply(A=Multiply(A=@n5.G, B=@n8), B=@n7.G))
  n10 = Transform(None=@n1) {transformSourceType=TRANSFORMSOURCE_World, transformType=TRANSFORM_Instance}
  n11 = Multiply(A=@n10, B=VectorParameter() {parameterName=MLW_ScaleMeshSize, defaultValue=<rgba 1 1 1 1>}.RGB)
  n12 = Transform(None=FunctionInput() {bUsePreviewValueAsDefault=true, inputName=WPO, sortPriority=2}) {transformSourceType=TRANSFORMSOURCE_World, transformType=TRANSFORM_Instance}
  n13 = VectorParameter() {parameterName=MLW_ScaleWPO, defaultValue=<rgba 1 1 1 1>}
  n14 = Add(
   A=Transform(None=Subtract(A=@n11, B=@n10)) {transformSourceType=TRANSFORMSOURCE_Instance}
   B=Transform(None=Multiply(A=@n12, B=@n13.RGB)) {transformSourceType=TRANSFORMSOURCE_Instance}
  )
  FunctionOutput(None=Add(A=Transform(None=@n3), B=Add(A=@n9, B=@n14))) {outputName=WPO, sortPriority=2}
  n15 = FunctionInput() {bUsePreviewValueAsDefault=true, inputName=Opacity, inputType=FunctionInput_Scalar, sortPriority=1, previewValue=<1 0 0 1>}
  n16 = ScalarParameter() {defaultValue=1, parameterName=MLW_ScaleOpacity}
  FunctionOutput(None=LinearInterpolate(A=Multiply(A=@n15, B=@n16), B=@n16, Alpha=ScalarParameter() {parameterName=MLW_EnableSpecificOpacity})) {outputName=Opacity, sortPriority=1}
  n17 = FunctionInput() {bUsePreviewValueAsDefault=true, inputName=Color, previewValue=<1 1 1 1>}
  n18 = VectorParameter() {parameterName=MLW_ScaleColor, defaultValue=<rgba 1 1 1 1>}
  FunctionOutput(None=LinearInterpolate(A=Multiply(A=@n17, B=@n18.RGB), B=@n18.RGB, Alpha=ScalarParameter() {parameterName=MLW_EnableSpecificColor})) {outputName=Color}
 ]
}
