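# M_Cap is a translucent unlit GiftedTraveler effect material for Niagara sprites, ribbons and mesh particles.
# A fiber texture, panned and UV-distorted by noise, drives emissive brightness, tinted by a two-color vertical gradient.
# Opacity combines vertical edge masks, an inverted Fresnel and the fiber texture, the latter two each blended in along V.
# Depth fade and particle alpha scale opacity, and MF_PrepareForMergeLWEmitter wraps the outputs.
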Material /GiftedTraveler/Assets/FX/Resources/M_Cap
graph {
 properties {
  bAutomaticallySetUsageInEditor=false
  bUsedWithNiagaraMeshParticles=true
  bUsedWithNiagaraRibbons=true
  bUsedWithNiagaraSprites=true
  blendMode=BLEND_Translucent
  shadingModel=MSM_Unlit
  usedShadingModels=Unlit
 }
 nodes [
  n1 = TextureObjectParameter() {group=MainTexture, parameterName=MainTexture_Texture, sortPriority=0, texture=/Game/Weapons/FORT_Melee/Pickaxe_MadameMoth/FX/T_Fibre_Noise}
  n2 = TextureObjectParameter() {group=NoiseA, parameterName=NoiseA_Texture, sortPriority=0, texture=/Game/Global/Textures/Default/Noise/T_Tiling_Noise_01}
  n3 = VectorParameter() {group=NoiseA, parameterName=NoiseA_UVControl, channelNames {a=VPan, b=UPan, g=VScale, r=UScale}, defaultValue=<rgba 1 0.125 0 0.03>}
  n4 = VectorParameter() {
   group=NoiseA
   parameterName=NoiseA_NoiseControl
   sortPriority=60
   channelNames {a=VStrength, b=UStrength, g=VOffset, r=UOffset}
   defaultValue=<rgba -0.5 -0.5 0 1>
  }
  n5 = VectorParameter() {group=MainTexture, parameterName=MainTexture_UVControl, channelNames {a=VPan, b=UPan, g=VScale, r=UScale}, defaultValue=<rgba 2 0.45 0 0.03>}
  n6 = MaterialFunctionCall(
   Texture=@n1
   AdditionalUVOffset=MaterialFunctionCall!(
    NoiseValue=MaterialFunctionCall(Texture=@n2, UVControl_RGB=@n3, UVControl_A=@n3) {materialFunction=&FXCommon_TextureSample}.RG
    Control_RGB=@n4
    Control_A=@n4
   ) {materialFunction=/Game/Effects/Niagara/MasterMaterials/Functions/Common/FXCommon_NoiseControl}.NoiseValue
   UVControl_RGB=@n5
   UVControl_A=@n5
  ) {materialFunction=&FXCommon_TextureSample}
  n7 = VectorParameter() {parameterName=MainTexture_RefineToColor, defaultValue=<rgba 2 4 0.5 0.5>}
  n8 = Add(A=MaterialFunctionCall(Texture=@n6.R, RGB=@n7.RGB, A=@n7.RGB) {materialFunction=&MF_FX_Refine}.Result)
  n9 = Multiply(
   A=MaterialFunctionCall(Float2=TextureCoordinate()) {materialFunction=&BreakOutFloat2Components}.G
   B=ScalarParameter() {defaultValue=1.7, parameterName=ColorGradMul}
  )
  n10 = ScalarParameter() {defaultValue=2, parameterName=ColorGradPower}
  n11 = LinearInterpolate(
   A=VectorParameter() {parameterName=Color1, defaultValue=<rgba 0.0099 0.0005 0.025 1>}.RGB
   B=VectorParameter() {parameterName=Color, defaultValue=<rgba 0.025 0.0005 0.0156 1>}.RGB
   Alpha=Saturate(None=Power(Base=@n9, Exp=@n10))
  )
  n12 = MaterialFunctionCall(Float2=TextureCoordinate()) {materialFunction=&BreakOutFloat2Components}
  n13 = OneMinus(None=@n12.G)
  n14 = VectorParameter() {group=ZGrad, parameterName=ZGrad_EdgeMask_InvertControl, channelNames {a="", b=Power, g=Multiply, r=Offset}, defaultValue=<rgba 0 1.5 1.2 1>}
  n15 = VectorParameter() {group=ZGrad, parameterName=ZGrad_EdgeMask_Control, channelNames {a="", b=Power, g=Multiply, r=Offset}, defaultValue=<rgba 1 10 1 1>}
  n16 = ScalarParameter() {defaultValue=3, parameterName=EX}
  n17 = Power(Base=OneMinus(None=Fresnel(ExponentIn=@n16) {baseReflectFraction=0}), Exp=ScalarParameter() {defaultValue=1.5, parameterName=FR})
  n18 = TextureCoordinate()
  n19 = VectorParameter() {
   group=FresnelGrad
   parameterName=FresnelGrad_EdgeMask_InvertControl
   channelNames {a="", b=Power, g=Multiply, r=Offset}
   defaultValue=<rgba 0 1 3 1>
  }
  n20 = VectorParameter() {
   group=FresnelGrad
   parameterName=FresnelGrad_EdgeMask_Control
   channelNames {a="", b=Power, g=Multiply, r=Offset}
   defaultValue=<rgba 1 1 1 1>
  }
  n21 = MaterialFunctionCall!(
   UorV=OneMinus(None=MaterialFunctionCall(Float2=@n18) {materialFunction=&BreakOutFloat2Components}.G)
   Invert_EdgeControl=@n19
   NoInvert_EdgeControl=@n20
  ) {materialFunction=/Game/Effects/Niagara/MasterMaterials/Functions/Common/FXCommon_EdgeMask_Ex}
  n22 = VectorParameter() {
   group=ZDisRe
   parameterName=ZDisRe_EdgeMask_InvertControl
   channelNames {a="", b=Power, g=Multiply, r=Offset}
   defaultValue=<rgba -0.0512 1 0.5 1>
  }
  n23 = VectorParameter() {group=ZDisRe, parameterName=ZDisRe_EdgeMask_Control, channelNames {a="", b=Power, g=Multiply, r=Offset}, defaultValue=<rgba 1 1.2159 1 1>}
  n24 = MaterialFunctionCall(UorV=@n12.G, Invert_EdgeControl=@n22, NoInvert_EdgeControl=@n23) {materialFunction=&FXCommon_EdgeMask_Ex}
  n25 = VectorParameter() {parameterName=MainTexture_Refine, defaultValue=<rgba 0.5 1.5 5 0.5>}
  n26 = Multiply(
   A=Multiply(
    A=MaterialFunctionCall(UorV=@n13, Invert_EdgeControl=@n14, NoInvert_EdgeControl=@n15) {materialFunction=&FXCommon_EdgeMask_Ex}.Mask
    B=LinearInterpolate(A=@n17, Alpha=@n21.Mask)
   )
   B=MaterialFunctionCall(Texture=LinearInterpolate(B=@n6.R, Alpha=@n24.Mask) {constA=1}, RGB=@n25.RGB, A=@n25.RGB) {materialFunction=&MF_FX_Refine}.Result
  )
  n27 = ScalarParameter() {defaultValue=50000, parameterName=DF}
  MaterialFunctionCall!(
   Color=MaterialFunctionCall(Emissive=Multiply(A=@n8, B=@n11)) {materialFunction=&MF_FX_EmissiveAdjustments}.Result
   Opacity=Multiply(A=ParticleColor().A, B=DepthFade(Opacity=@n26, FadeDistance=@n27))
   WPO
  ) {materialFunction=/Game/Effects/Fort_Effects/Materials/_Functions/MF_PrepareForMergeLWEmitter}
 ]
}
