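# Translucent, two-sided unlit Niagara material for an upward-flowing cap effect.
# Emissive adds refined, panning dust detail tinted by SandC to particle color hue-shifted along V.
# Opacity multiplies particle alpha, a V-axis tip/tail edge mask, power-warped panning plasma and inverted Fresnel that softens silhouettes.
# Squared camera-distance fade and depth fade further attenuate opacity before output through MF_PrepareForMergeLWEmitter; the FlowC parameter is unconnected.
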Material /GiftedTraveler/Assets/FX/Resources/M_FlowUpCap
graph {
 properties {
  bAutomaticallySetUsageInEditor=false
  bUsedWithNiagaraMeshParticles=true
  bUsedWithNiagaraRibbons=true
  bUsedWithNiagaraSprites=true
  blendMode=BLEND_Translucent
  shadingModel=MSM_Unlit
  twoSided=true
  usedShadingModels=Unlit
 }
 nodes [
  n1 = TextureObjectParameter() {group=Detail, parameterName=Detail_Texture, sortPriority=0, texture=/Game/Global/Textures/Default/PointSplatter/T_Tiling_Dust}
  n2 = VectorParameter() {group=Detail, parameterName=Detail_UVControl, channelNames {a=VPan, b=UPan, g=VScale, r=UScale}, defaultValue=<rgba 1 1 0 -0.5>}
  n3 = MaterialFunctionCall(Texture=@n1, UVControl_RGB=@n2, UVControl_A=@n2) {materialFunction=&FXCommon_TextureSample}
  n4 = VectorParameter() {parameterName=Detail_Refine, defaultValue=<rgba 2 10 0.5 0.5>}
  n5 = MaterialFunctionCall(Float2=TextureCoordinate()) {materialFunction=&BreakOutFloat2Components}
  n6 = ScalarParameter() {defaultValue=-0.2, parameterName=HUES}
  n7 = ParticleColor()
  n8 = Add(
   A=Multiply(
    A=MaterialFunctionCall(Texture=@n3.R, RGB=@n4.RGB, A=@n4.RGB) {materialFunction=&MF_FX_Refine}.Result
    B=VectorParameter() {parameterName=SandC, defaultValue=<rgba 0 21.2299 100 1>}.RGB
   )
   B=MaterialFunctionCall(Hue Shift Percentage=Multiply(A=@n5.G, B=@n6), Texture=@n7.RGB) {materialFunction=&HueShift}.Result
  )
  n9 = MaterialFunctionCall(Float2=TextureCoordinate()) {materialFunction=&BreakOutFloat2Components}
  n10 = VectorParameter() {
   group=TIPTAILMASK
   parameterName=TIPTAILMASK_EdgeMask_InvertControl
   channelNames {a="", b=Power, g=Multiply, r=Offset}
   defaultValue=<rgba 0 1.5 3 1>
  }
  n11 = VectorParameter() {
   group=TIPTAILMASK
   parameterName=TIPTAILMASK_EdgeMask_Control
   channelNames {a="", b=Power, g=Multiply, r=Offset}
   defaultValue=<rgba 0 50 1 1>
  }
  n12 = TextureObjectParameter() {
   group=MainFlowTexture
   parameterName=MainFlowTexture_Texture
   samplerType=SAMPLERTYPE_LinearGrayscale
   sortPriority=0
   texture=/Game/Effects/Fort_Effects/Emotes/UltraLight/Texture/T_Plasma_B_Greyscale
  }
  n13 = VectorParameter() {
   group=MainFlowTexture
   parameterName=MainFlowTexture_UVControl
   channelNames {a=VPan, b=UPan, g=VScale, r=UScale}
   defaultValue=<rgba 1 0.2 0 -0.5>
  }
  n14 = MaterialFunctionCall(Float2=TextureCoordinate()) {materialFunction=&BreakOutFloat2Components}
  n15 = Power(Base=@n14.G, Exp=ScalarParameter() {defaultValue=1.5, parameterName=UVPow})
  n16 = VectorParameter() {parameterName=MainFlowTexture_Refine, defaultValue=<rgba 4 6 0.35 0.5>}
  n17 = MaterialFunctionCall(
   Texture=MaterialFunctionCall(
    Texture=@n12
    UVControl_RGB=@n13
    UVControl_A=@n13
    "[Override]UVInput"=MaterialFunctionCall(X=@n14.R, Y=@n15) {materialFunction=&MakeFloat2}.Result
   ) {materialFunction=&FXCommon_TextureSample}.R
   RGB=@n16.RGB
   A=@n16.RGB
  ) {materialFunction=&MF_FX_Refine}
  n18 = ScalarParameter() {defaultValue=1, parameterName=f}
  n19 = Power(Base=OneMinus(None=Fresnel(ExponentIn=@n18) {baseReflectFraction=0}), Exp=ScalarParameter() {defaultValue=5, parameterName=z})
  n20 = Multiply(
   A=@n7.A
   B=Multiply(
    A=MaterialFunctionCall!(UorV=@n9.G, Invert_EdgeControl=@n10, NoInvert_EdgeControl=@n11) {materialFunction=/Game/Effects/Niagara/MasterMaterials/Functions/Common/FXCommon_EdgeMask_Ex}.Mask
    B=Multiply(A=@n17.Result, B=@n19)
   )
  )
  n21 = MaterialFunctionCall!(Distance=ScalarParameter() {defaultValue=40000, parameterName=FadeDistance}) {materialFunction=&/MF_CameraDistanceFade}
  n22 = Multiply(A=@n21.Result, B=@n21.Result)
  n23 = ScalarParameter() {defaultValue=10000, parameterName=DepthFade}
  MaterialFunctionCall!(
   Color=MaterialFunctionCall(Emissive=Divide(A=@n8) {constB=2.2711}) {materialFunction=&MF_FX_EmissiveAdjustments}.Result
   Opacity=Multiply(A=Multiply(A=@n20, B=@n22), B=DepthFade(FadeDistance=@n23))
   WPO
  ) {materialFunction=/Game/Effects/Fort_Effects/Materials/_Functions/MF_PrepareForMergeLWEmitter}
  VectorParameter() {parameterName=FlowC, defaultValue=<rgba 20 2.4696 8.1402 1>}
 ]
}
