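# Translucent, unlit, two-sided Niagara material for a glowing ray effect, with color, opacity and vertex offset driven by dynamic parameters.
# Two scrolling noises distort a plasma texture on remapped UVs, tinted by a lengthwise two-color gradient.
# Opacity combines edge masks, axis-aligned fresnel, depth, camera-distance and world-height fades.
# Sphere masks locally modulate emissive and normal-direction swelling, while masked tip/tail offsets add to WPO through MF_PrepareForMergeLWEmitter.
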Material /GiftedTraveler/Assets/FX/Resources/M_RayMaterial
graph {
 properties {
  bAutomaticallySetUsageInEditor=false
  bUsedWithNiagaraMeshParticles=true
  bUsedWithNiagaraRibbons=true
  bUsedWithNiagaraSprites=true
  blendMode=BLEND_Translucent
  shadingModel=MSM_Unlit
  twoSided=true
  usedShadingModels=Unlit
 }
 nodes [
  n1 = DynamicParameter() {parameterIndex=1, paramNames [LS,WPOS,EndFade,Param4]}
  n2 = WorldPosition()
  n3 = DynamicParameter() {paramNames [X,Y,Z,Param4]}
  n4 = LinearInterpolate(
   B=Multiply(A=@n1.LS, B=ScalarParameter() {defaultValue=1, parameterName=DPLSSCALE})
   Alpha=SphereMask(
    A=@n2.XYZ
    B=MaterialFunctionCall(X=@n3.X, Y=@n3.X, Z=@n3.X) {materialFunction=&MakeFloat3}.Result
    Radius=@n3.Param4
   ) {hardnessPercent=0.1}
  ) {constA=1}
  n5 = MaterialFunctionCall(Float2=TextureCoordinate()) {materialFunction=&BreakOutFloat2Components}
  n6 = VectorParameter() {group=TipTailWPO, parameterName=ZColorGrad_EdgeMask_Control, channelNames {a="", b=Power, g=Multiply, r=Offset}, defaultValue=<rgba 0 1 3 1>}
  n7 = Saturate(None=Multiply(A=Add(A=@n5.G, B=@n6.Offset), B=@n6.Multiply))
  n8 = VectorParameter() {group=ColorTexture, parameterName=ColorTexture_UVControl, channelNames {a=VPan, b=UPan, g=VScale, r=UScale}, defaultValue=<rgba 1 1 0 0>}
  n9 = Desaturation(
   None=MaterialFunctionCall(
    Texture=TextureObjectParameter() {group=ColorTexture, parameterName=ColorTexture_Texture, sortPriority=0}
    UVControl_RGB=@n8
    UVControl_A=@n8
   ) {materialFunction=&FXCommon_TextureSample}.RGB
   Fraction=ScalarParameter() {parameterName=ColorTexture_Desaturate}
  )
  n10 = ScalarParameter() {defaultValue=1, parameterName=ColorTexture_Power}
  n11 = Add(
   A=LinearInterpolate(
    A=VectorParameter() {parameterName=SecondColor, defaultValue=<rgba 4.3146 3 30 1>}.RGB
    B=VectorParameter() {parameterName=MainColor, defaultValue=<rgba 50 5 12.5 1>}.RGB
    Alpha=Power(Base=@n7, Exp=@n6.Power)
   )
   B=Multiply(A=Power(Base=@n9, Exp=@n10), B=ScalarParameter() {parameterName=Color_Texture_Scale})
  )
  n12 = TextureObjectParameter() {
   group=MainTexture
   parameterName=MainTexture_Texture
   samplerType=SAMPLERTYPE_LinearGrayscale
   sortPriority=0
   texture=/Game/Effects/Fort_Effects/Emotes/UltraLight/Texture/T_Plasma_A_B_Greyscale
  }
  n13 = TextureObjectParameter() {
   group=NoiseA
   parameterName=NoiseA_Texture
   samplerType=SAMPLERTYPE_LinearColor
   sortPriority=0
   texture=/Engine/Functions/Engine_MaterialFunctions02/ExampleContent/Textures/OffsetNoiseDistanceFields_ForNormals
  }
  n14 = VectorParameter() {group=NoiseA, parameterName=NoiseA_UVControl, channelNames {a=VPan, b=UPan, g=VScale, r=UScale}, defaultValue=<rgba 0.5 1 3 5>}
  n15 = NamedRerouteUsage()
  n16 = NamedRerouteUsage()
  n17 = MaterialFunctionCall(Texture=@n13, UVControl_RGB=@n14, UVControl_A=@n14, "[Override]TimeInput"=@n15, "[Override]UVInput"=@n16) {materialFunction=&FXCommon_TextureSample}
  n18 = VectorParameter() {
   group=NoiseA
   parameterName=NoiseA_NoiseControl
   sortPriority=60
   channelNames {a=VStrength, b=UStrength, g=VOffset, r=UOffset}
   defaultValue=<rgba -0.5 -0.5 0 0>
  }
  n19 = TextureObjectParameter() {
   group=NoiseB
   parameterName=NoiseB_Texture
   samplerType=SAMPLERTYPE_LinearGrayscale
   sortPriority=0
   texture=/Game/Effects/Fort_Effects/Emotes/UltraLight/Texture/T_Cloud_Greyscale
  }
  n20 = VectorParameter() {group=NoiseB, parameterName=NoiseB_UVControl, channelNames {a=VPan, b=UPan, g=VScale, r=UScale}, defaultValue=<rgba 3 3 0 1>}
  n21 = MaterialFunctionCall(Texture=@n19, UVControl_RGB=@n20, UVControl_A=@n20, "[Override]TimeInput"=@n15, "[Override]UVInput"=@n16) {materialFunction=&FXCommon_TextureSample}
  n22 = VectorParameter() {
   group=NoiseB
   parameterName=NoiseB_NoiseControl
   sortPriority=60
   channelNames {a=VStrength, b=UStrength, g=VOffset, r=UOffset}
   defaultValue=<rgba -0.5 -0.5 0 0>
  }
  n23 = VectorParameter() {
   group=MainTexture
   parameterName=MainTexture_UVControl
   channelNames {a=VPan, b=UPan, g=VScale, r=UScale}
   defaultValue=<rgba 0.6905 0.2 0 3>
  }
  n24 = MaterialFunctionCall(
   Texture=@n12
   AdditionalUVOffset=Add(
    A=MaterialFunctionCall!(NoiseValue=@n17.R, Control_RGB=@n18, Control_A=@n18) {materialFunction=/Game/Effects/Niagara/MasterMaterials/Functions/Common/FXCommon_NoiseControl}.NoiseValue
    B=MaterialFunctionCall(NoiseValue=@n21.RG, Control_RGB=@n22, Control_A=@n22) {materialFunction=&FXCommon_NoiseControl}.NoiseValue
   )
   UVControl_RGB=@n23
   UVControl_A=@n23
   "[Override]TimeInput"=NamedRerouteUsage()
   "[Override]UVInput"=@n16
  ) {materialFunction=&FXCommon_TextureSample}
  n25 = VectorParameter() {parameterName=MainTexture_RefineToColor, defaultValue=<rgba 2 20 10.1775 0.5>}
  n26 = Add(A=MaterialFunctionCall(Texture=@n24.R, RGB=@n25.RGB, A=@n25.RGB) {materialFunction=&MF_FX_Refine}.Result)
  n27 = ParticleColor()
  n28 = VectorParameter() {parameterName=FinalScale, channelNames {a="", b="", g=Alpha, r=Color}, defaultValue=<rgba 1 1 0 0>}
  n29 = Multiply(A=Multiply(A=Multiply(A=@n11, B=@n26), B=@n27.RGB), B=@n28.Color)
  n30 = MaterialFunctionCall!(Distance=ScalarParameter() {defaultValue=40000, parameterName=FadeDistance}) {materialFunction=&/MF_CameraDistanceFade}
  n31 = VectorParameter() {parameterName=StartUP_TipFade, defaultValue=<rgba -5000 10000 0 1>}
  n32 = Add(
   A=Subtract(
    A=@n3.Z
    B=MaterialFunctionCall(Float3=@n2.XYZ) {materialFunction=&BreakOutFloat3Components}.B
   )
   B=@n31.R
  )
  n33 = VectorParameter() {group=ZGrad, parameterName=ZGrad_EdgeMask_InvertControl, channelNames {a="", b=Power, g=Multiply, r=Offset}, defaultValue=<rgba 0 100 2 1>}
  n34 = VectorParameter() {group=ZGrad, parameterName=ZGrad_EdgeMask_Control, channelNames {a="", b=Power, g=Multiply, r=Offset}, defaultValue=<rgba 0 2.0375 1.5 1>}
  n35 = MaterialFunctionCall!(
   UorV=MaterialFunctionCall(Float2=TextureCoordinate()) {materialFunction=&BreakOutFloat2Components}.G
   Invert_EdgeControl=@n33
   NoInvert_EdgeControl=@n34
  ) {materialFunction=/Game/Effects/Niagara/MasterMaterials/Functions/Common/FXCommon_EdgeMask_Ex}
  n36 = MaterialFunctionCall(Axis=Constant3Vector() {constant=<rgba 0 0 1 0>}, Exponent=ScalarParameter() {defaultValue=5, parameterName=FresnelExponent}) {materialFunction=&AxisAlignedFresnel}
  n37 = VectorParameter() {parameterName=MainTexture_RefineToOpacity, defaultValue=<rgba 8.5581 2 0.01 0.5>}
  n38 = Multiply(
   A=Multiply(A=@n27.A, B=@n28.Alpha)
   B=Multiply(
    A=Multiply(A=@n35.Mask, B=@n36.Fresnel)
    B=MaterialFunctionCall(Texture=@n24.R, RGB=@n37.RGB, A=@n37.RGB) {materialFunction=&MF_FX_Refine}.Result
   )
  )
  n39 = ScalarParameter() {defaultValue=50000, parameterName=DepthFade}
  n40 = ScalarParameter() {defaultValue=1, parameterName=LowestC}
  n41 = VectorParameter() {
   group=CenterMask
   parameterName=CenterMask_EdgeMask_InvertControl
   channelNames {a="", b=Power, g=Multiply, r=Offset}
   defaultValue=<rgba 0 1 1 1>
  }
  n42 = VectorParameter() {group=CenterMask, parameterName=CenterMask_EdgeMask_Control, channelNames {a="", b=Power, g=Multiply, r=Offset}, defaultValue=<rgba 0 1 1 1>}
  n43 = MaterialFunctionCall(
   UorV=MaterialFunctionCall(Float2=TextureCoordinate()) {materialFunction=&BreakOutFloat2Components}.G
   Invert_EdgeControl=@n41
   NoInvert_EdgeControl=@n42
  ) {materialFunction=&FXCommon_EdgeMask_Ex}
  n44 = Multiply(
   A=Saturate(None=Divide(A=@n32, B=@n31.G))
   B=Multiply(A=DepthFade(Opacity=@n38, FadeDistance=@n39), B=LinearInterpolate(B=@n40, Alpha=@n43.Mask) {constA=1})
  )
  n45 = Add(
   A=Subtract(A=WorldPosition().Z, B=DynamicParameter() {parameterIndex=1, paramNames [LS,WPOS,EndFade,Param4]}.EndFade)
   B=ScalarParameter() {parameterName=EndFade_Offset}
  )
  n46 = ScalarParameter() {defaultValue=130000, parameterName=EndFade_Hardness}
  n47 = Power(Base=Saturate(None=Divide(A=@n45, B=@n46)), Exp=ScalarParameter() {defaultValue=2, parameterName=EndFade_Power})
  n48 = Multiply(
   A=TransformPosition(None=@n2.XYZ) {transformSourceType=TRANSFORMPOSSOURCE_World, transformType=TRANSFORMPOSSOURCE_Instance}
   B=VectorParameter() {parameterName=MLW_ScaleMeshSize, defaultValue=<rgba 1 1 1 1>}.RGB
  )
  n49 = LinearInterpolate(
   B=Multiply(A=@n1.WPOS, B=ScalarParameter() {defaultValue=1, parameterName=DPWPOSSCALE})
   Alpha=SphereMask(
    A=TransformPosition(None=@n48) {transformSourceType=TRANSFORMPOSSOURCE_Instance, transformType=TRANSFORMPOSSOURCE_World}
    B=MaterialFunctionCall(X=@n3.X, Y=@n3.X, Z=@n3.X) {materialFunction=&MakeFloat3}.Result
    Radius=@n3.Param4
   ) {hardnessPercent=0.1}
  )
  n50 = Constant()
  n51 = Length(None=Transform(None=MaterialFunctionCall(X=@n49, Y=@n50, Z=@n50) {materialFunction=&MakeFloat3}.Result) {transformSourceType=TRANSFORMSOURCE_Local}) {is=class MaterialExpressionLength}
  n52 = VertexNormalWS()
  n53 = VectorParameter() {parameterName=StartUP_WPOMask, defaultValue=<rgba -5000 10000 0 1>}
  n54 = Add(A=Subtract(A=@n3.Z, B=@n53.R), B=@n53.G)
  n55 = Multiply(A=Saturate(None=Divide(A=@n54)), B=NamedRerouteUsage())
  MaterialFunctionCall!(
   Color=MaterialFunctionCall(Emissive=Multiply(A=@n4, B=@n29)) {materialFunction=&MF_FX_EmissiveAdjustments}.Result
   Opacity=Multiply(A=Multiply(A=@n30.Result, B=@n30.Result), B=Multiply(A=@n44, B=@n47))
   WPO=Add(A=Multiply(A=@n51, B=@n52), B=Multiply(A=@n55, B=@n47))
  ) {materialFunction=/Game/Effects/Fort_Effects/Materials/_Functions/MF_PrepareForMergeLWEmitter}
  n56 = MaterialFunctionCall(Float2=TextureCoordinate()) {materialFunction=&BreakOutFloat2Components}
  n57 = VectorParameter() {
   group=TipTailWPO
   parameterName=TipTailWPO_EdgeMask_Control
   channelNames {a="", b=Power, g=Multiply, r=Offset}
   defaultValue=<rgba 0 1 45.8349 1>
  }
  n58 = Add(A=@n56.G, B=@n57.Offset)
  n59 = Power(Base=Saturate(None=Multiply(A=@n58, B=@n57.Multiply)), Exp=@n57.Power)
  n60 = Constant()
  n61 = MaterialFunctionCall(
   X=Multiply(
    A=DynamicParameter() {parameterIndex=2, paramNames [ScaleTipTailWPO, Param2, Param3, Param4]}.ScaleTipTailWPO
    B=ScalarParameter() {defaultValue=44.8843, parameterName=TipTailWPO_Str}
   )
   Y=@n60
   Z=@n60
  ) {materialFunction=&MakeFloat3}
  n62 = Multiply(A=VertexNormalWS(), B=Length(None=Transform(None=@n61.Result) {transformSourceType=TRANSFORMSOURCE_Local}))
  n63 = VectorParameter() {
   group=TipTailWPO
   parameterName=TipTailWPO_EdgeMask_InvertControl
   channelNames {a="", b=Power, g=Multiply, r=Offset}
   defaultValue=<rgba -2 1 2 1>
  }
  n64 = Add(A=OneMinus(None=@n56.G), B=@n63.Offset)
  n65 = Power(Base=Saturate(None=Multiply(A=@n64, B=@n63.Multiply)), Exp=@n63.Power)
  NamedRerouteDeclaration(None=Add(A=Multiply(A=@n59, B=@n62), B=Multiply(A=@n65, B=@n62))) {name=TipTailWPO, nodeColor=<rgba 1 0.2812 0 1>}
  NamedRerouteDeclaration(None=Multiply(A=Time(), B=ScalarParameter() {defaultValue=0.01, parameterName=ScaleTime})) {name=ScaledTime, nodeColor=<rgba 0.3594 1 0 1>}
  n66 = MaterialFunctionCall(Float2=TextureCoordinate()) {materialFunction=&BreakOutFloat2Components}
  n67 = Power(Base=OneMinus(None=@n66.G), Exp=ScalarParameter() {defaultValue=0.4, parameterName=UVZPower})
  NamedRerouteDeclaration(None=MaterialFunctionCall(X=@n66.R, Y=OneMinus(None=@n67)) {materialFunction=&MakeFloat2}.Result) {name=ScaledUV, nodeColor=<rgba 0 0.6016 1 1>}
  OneMinus(None=@n5.G)
 ]
}
