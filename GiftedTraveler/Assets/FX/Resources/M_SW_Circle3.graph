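# M_SW_Circle3 is a translucent, unlit, two-sided Niagara effect material for an animated, glowing circular streak with refraction.
# An edge mask and a panning, axis-swapped smoke texture shape opacity and refraction, with vertex red offsetting the pan time.
# Noise-distorted detail and plasma layers form the tinted emissive, blended toward a hue-shifted tip color.
# Dynamic parameters slide the growth along V, scale tip brightness and drive SplineThicken width offsets.
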Material /GiftedTraveler/Assets/FX/Resources/M_SW_Circle3
graph {
 properties {
  bAutomaticallySetUsageInEditor=false
  bUsedWithNiagaraMeshParticles=true
  bUsedWithNiagaraRibbons=true
  bUsedWithNiagaraSprites=true
  blendMode=BLEND_Translucent
  refractionMethod=RM_PixelNormalOffset
  shadingModel=MSM_Unlit
  twoSided=true
  usedShadingModels=Unlit
 }
 nodes [
  n1 = ScalarParameter() {defaultValue=1.0444, parameterName=ref}
  n2 = VectorParameter() {
   group=TipTailFade
   parameterName=TipTailFade_EdgeMask_InvertControl
   channelNames {a="", b=Power, g=Multiply, r=Offset}
   defaultValue=<rgba 0 8 1 1>
  }
  n3 = VectorParameter() {
   group=TipTailFade
   parameterName=TipTailFade_EdgeMask_Control
   channelNames {a="", b=Power, g=Multiply, r=Offset}
   defaultValue=<rgba 0 1.265 2 1>
  }
  n4 = MaterialFunctionCall!(
   UorV=MaterialFunctionCall(Float2=TextureCoordinate()) {materialFunction=&BreakOutFloat2Components}.G
   Invert_EdgeControl=@n2
   NoInvert_EdgeControl=@n3
  ) {materialFunction=/Game/Effects/Niagara/MasterMaterials/Functions/Common/FXCommon_EdgeMask_Ex}
  n5 = TextureObjectParameter() {
   group=CircleTexture
   parameterName=CircleTexture_Texture
   samplerType=SAMPLERTYPE_Grayscale
   sortPriority=0
   texture=&T_Smoke_Trail_Soft
  }
  n6 = VectorParameter() {group=CircleTexture, parameterName=CircleTexture_UVControl, channelNames {a=VPan, b=UPan, g=VScale, r=UScale}, defaultValue=<rgba 1 1 0.3 0>}
  n7 = Add(
   A=Multiply(A=VertexColor().R, B=ScalarParameter() {defaultValue=85077.8672, parameterName=TS})
   B=Multiply(A=Time() {bOverride_Period=true, period=6000}, B=ScalarParameter() {defaultValue=0.35, parameterName=wd})
  )
  n8 = MaterialFunctionCall(
   Texture=@n5
   UVControl_RGB=@n6
   UVControl_A=@n6
   "[Override]TimeInput"=@n7
   "[Override]UVInput"=MaterialFunctionCall(XY=TextureCoordinate()) {materialFunction=&Swizzle}.YX
  ) {materialFunction=&FXCommon_TextureSample}
  n9 = Multiply(A=@n4.Mask, B=@n8.R)
  n10 = VectorParameter() {parameterName=CircleTexture_Refine, defaultValue=<rgba 5 2 1 0.5>}
  n11 = Multiply(B=Multiply(
    A=@n4.Mask
    B=MaterialFunctionCall(Texture=@n8.R, RGB=@n10.RGB, A=@n10.RGB) {materialFunction=&MF_FX_Refine}.Result
   )) {constA=1}
  n12 = ParticleColor()
  LinearInterpolate(B=LinearInterpolate(B=LinearInterpolate(B=@n1, Alpha=@n9) {constA=1}, Alpha=@n11) {constA=1}, Alpha=@n12.A) {constA=1}
  n13 = DynamicParameter() {paramNames [VOffset,VMul,VBright,Param4]}
  n14 = ScalarParameter() {defaultValue=36383.8438, parameterName=WidthBase}
  n15 = ScalarParameter() {defaultValue=8249.4961, parameterName=WidthTip}
  n16 = Add(
   A=MaterialFunctionCall(Float2=TextureCoordinate()) {materialFunction=&BreakOutFloat2Components}.G
   B=@n13.VOffset
  )
  n17 = Multiply(A=@n16, B=@n13.VMul)
  LinearInterpolate(
   B=Multiply(
    A=@n13.Param4
    B=MaterialFunctionCall(WidthBase=@n14, WidthTip=@n15) {materialFunction=&SplineThicken}.WorldPositionOffset
   )
   Alpha=Power(Base=Saturate(None=@n17)) {constExponent=0.5}
  )
  n18 = TextureObjectParameter() {
   group=DetailTexture2
   parameterName=DetailTexture2_Texture
   sortPriority=0
   texture=/Game/CharClasses/Outlander/Gadgets/AntiMaterialCharge/Materials/T_AntiMaterialCharge_Dissolve_01
  }
  n19 = TextureObjectParameter() {
   group=Noise
   parameterName=Noise_Texture
   samplerType=SAMPLERTYPE_LinearGrayscale
   sortPriority=0
   texture=&T_SimplePerlin_Greyscale
  }
  n20 = VectorParameter() {group=Noise, parameterName=Noise_UVControl, channelNames {a=VPan, b=UPan, g=VScale, r=UScale}, defaultValue=<rgba 0.5 1 0 0.1>}
  n21 = VectorParameter() {
   group=Noise
   parameterName=Noise_NoiseControl
   sortPriority=60
   channelNames {a=VStrength, b=UStrength, g=VOffset, r=UOffset}
   defaultValue=<rgba -0.5 -0.5 0 -0.02>
  }
  n22 = MaterialFunctionCall!(
   NoiseValue=MaterialFunctionCall(Texture=@n19, UVControl_RGB=@n20, UVControl_A=@n20) {materialFunction=&FXCommon_TextureSample}.RG
   Control_RGB=@n21
   Control_A=@n21
  ) {materialFunction=/Game/Effects/Niagara/MasterMaterials/Functions/Common/FXCommon_NoiseControl}
  n23 = VectorParameter() {
   group=DetailTexture2
   parameterName=DetailTexture2_UVControl
   channelNames {a=VPan, b=UPan, g=VScale, r=UScale}
   defaultValue=<rgba 1 1.5 0 0.2>
  }
  n24 = VectorParameter() {parameterName=DetailTexture2_Refine, defaultValue=<rgba 2 2 0 0>}
  n25 = MaterialFunctionCall(
   Texture=MaterialFunctionCall(Texture=@n18, AdditionalUVOffset=@n22.NoiseValue, UVControl_RGB=@n23, UVControl_A=@n23, "[Override]TimeInput"=@n7) {materialFunction=&FXCommon_TextureSample}.RGB
   RGB=@n24.RGB
   A=@n24.RGB
  ) {materialFunction=&MF_FX_Refine}
  n26 = TextureObjectParameter() {group=DetailTexture, parameterName=DetailTexture_Texture, sortPriority=0, texture=/Game/Effects/Fort_Effects/Textures/Tile/T_TileNoise_01}
  n27 = VectorParameter() {
   group=DetailTexture
   parameterName=DetailTexture_UVControl
   channelNames {a=VPan, b=UPan, g=VScale, r=UScale}
   defaultValue=<rgba 1 10 0 0.03>
  }
  n28 = VectorParameter() {parameterName=DetailTexture_Refine, defaultValue=<rgba 2.5 500 0 0.5>}
  n29 = MaterialFunctionCall(
   Texture=MaterialFunctionCall(Texture=@n26, AdditionalUVOffset=@n22.NoiseValue, UVControl_RGB=@n27, UVControl_A=@n27, "[Override]TimeInput"=@n7) {materialFunction=&FXCommon_TextureSample}.RGB
   RGB=@n28.RGB
   A=@n28.RGB
  ) {materialFunction=&MF_FX_Refine}
  n30 = TextureObjectParameter() {
   group=Rainbow
   parameterName=Rainbow_Texture
   samplerType=SAMPLERTYPE_LinearGrayscale
   sortPriority=0
   texture=/Game/Effects/Fort_Effects/Emotes/UltraLight/Texture/T_Plasma_B_Greyscale
  }
  n31 = VectorParameter() {group=Rainbow, parameterName=Rainbow_UVControl, channelNames {a=VPan, b=UPan, g=VScale, r=UScale}, defaultValue=<rgba 0.2808 0.3784 0 0.05>}
  n32 = VectorParameter() {parameterName=Rainbow_Refine, defaultValue=<rgba 15 81.0861 2 0.5>}
  n33 = MaterialFunctionCall(
   Texture=Desaturation(
    None=MaterialFunctionCall(Texture=@n30, AdditionalUVOffset=@n22.NoiseValue, UVControl_RGB=@n31, UVControl_A=@n31, "[Override]TimeInput"=@n7) {materialFunction=&FXCommon_TextureSample}.RGB
    Fraction=ScalarParameter() {parameterName=f}
   )
   RGB=@n32.RGB
   A=@n32.RGB
  ) {materialFunction=&MF_FX_Refine}
  n34 = ScalarParameter() {defaultValue=1, parameterName=aeta}
  n35 = Add(
   A=Multiply(
    A=Multiply(A=@n25.Result, B=@n29.Result)
    B=VectorParameter() {parameterName=DetailTextureColor, defaultValue=<rgba 0.0286 0.0183 1.5 1>}.RGB
   )
   B=Multiply(A=Multiply(A=@n33.Result, B=@n34), B=VectorParameter() {parameterName=C, defaultValue=<rgba 0.9532 0.7 3.5 1>}.RGB)
  )
  n36 = MaterialFunctionCall(Float2=TextureCoordinate()) {materialFunction=&BreakOutFloat2Components}
  n37 = Constant3Vector() {constant=<rgba 0 2 1 0>}
  n38 = VectorParameter() {parameterName=TipC_Refine, defaultValue=<rgba 13 50 1 -0.25>}
  n39 = Multiply(
   A=Power(
    Base=MaterialFunctionCall!(UorV=@n36.R, EdgeControl=@n37) {materialFunction=/Game/Effects/Niagara/MasterMaterials/Functions/Common/FXCommon_EdgeMask_Simple}.Mask
    Exp=@n38.G
   )
   B=@n38.R
  )
  n40 = Multiply(A=@n38.A, B=Power(Base=OneMinus(None=@n36.G)) {constExponent=0.8})
  n41 = VectorParameter() {parameterName=TipC, defaultValue=<rgba 13 0.975 7.9896 1>}
  n42 = Multiply(
   A=Multiply(
    A=Add(A=@n39)
    B=MaterialFunctionCall(Hue Shift Percentage=@n40, Texture=@n41.RGB) {materialFunction=&HueShift}.Result
   ) {constB=10}
   B=@n35
  )
  n43 = DynamicParameter() {paramNames [VOffset,VMul,VBright,Param4]}
  n44 = VectorParameter() {parameterName=Growth_P, defaultValue=<rgba 1.6807 28.2576 5 1>}
  n45 = Saturate(None=Multiply(A=@n16, B=@n44.B))
  n46 = VectorParameter() {parameterName=Growth_Refine, defaultValue=<rgba 6 9 0.1 0.5>}
  n47 = Multiply(
   A=@n43.VBright
   B=MaterialFunctionCall(Texture=OneMinus(None=@n45), RGB=@n46.RGB, A=@n46.RGB) {materialFunction=&MF_FX_Refine}.Result
  )
  MaterialFunctionCall(Emissive=Divide(A=LinearInterpolate(A=@n35, B=@n42, Alpha=@n47)) {constB=2.2711}) {materialFunction=&MF_FX_EmissiveAdjustments}
  Multiply(A=Multiply(A=@n11, B=@n12.A), B=Saturate(None=Multiply(A=@n16, B=@n44.G)))
  n48 = MaterialFunctionCall(Float2=TextureCoordinate()) {materialFunction=&BreakOutFloat2Components}
  Saturate(None=MaterialFunctionCall(In=Add(A=@n48.G, B=@n43.VOffset), Contrast=Constant() {r=1}) {materialFunction=&CheapContrast}.Result)
  ScalarParameter() {defaultValue=-0.8824, parameterName=s}
 ]
}
