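# M_SW_Circle4 is a translucent, two-sided unlit Niagara material for animated circle or streak effects with noise-warped rainbow color.
# Perlin noise distorts the rainbow texture, which is desaturated and tinted for emissive, with panning time offset by vertex red.
# A swizzled-UV circle texture and V-edge fade build masks for opacity and refraction, modulated by dynamic VOffset and particle alpha.
# The graph also includes a SplineThicken call for width control.
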Material /GiftedTraveler/Assets/FX/Resources/M_SW_Circle4
graph {
 properties {
  bAutomaticallySetUsageInEditor=false
  bUsedWithNiagaraMeshParticles=true
  bUsedWithNiagaraRibbons=true
  bUsedWithNiagaraSprites=true
  blendMode=BLEND_Translucent
  refractionMethod=RM_PixelNormalOffset
  shadingModel=MSM_Unlit
  twoSided=true
  usedShadingModels=Unlit
 }
 nodes [
  n1 = TextureObjectParameter() {group=Rainbow, parameterName=Rainbow_Texture, sortPriority=0, texture=&T_Tile_RainbowColor}
  n2 = TextureObjectParameter() {
   group=Noise
   parameterName=Noise_Texture
   samplerType=SAMPLERTYPE_LinearGrayscale
   sortPriority=0
   texture=&T_SimplePerlin_Greyscale
  }
  n3 = VectorParameter() {group=Noise, parameterName=Noise_UVControl, channelNames {a=VPan, b=UPan, g=VScale, r=UScale}, defaultValue=<rgba 0.5 1 0 0.1>}
  n4 = MaterialFunctionCall(Texture=@n2, UVControl_RGB=@n3, UVControl_A=@n3) {materialFunction=&FXCommon_TextureSample}
  n5 = VectorParameter() {
   group=Noise
   parameterName=Noise_NoiseControl
   sortPriority=60
   channelNames {a=VStrength, b=UStrength, g=VOffset, r=UOffset}
   defaultValue=<rgba -0.5 -0.5 0 -0.02>
  }
  n6 = VectorParameter() {group=Rainbow, parameterName=Rainbow_UVControl, channelNames {a=VPan, b=UPan, g=VScale, r=UScale}, defaultValue=<rgba 0.2808 0.3784 0 0.05>}
  n7 = Add(
   A=Multiply(A=VertexColor().R, B=ScalarParameter() {defaultValue=182002.4531, parameterName=TS})
   B=Multiply(A=Time(), B=ScalarParameter() {defaultValue=0.35, parameterName=wd})
  )
  n8 = Desaturation(
   None=MaterialFunctionCall(
    Texture=@n1
    AdditionalUVOffset=MaterialFunctionCall!(NoiseValue=@n4.RG, Control_RGB=@n5, Control_A=@n5) {materialFunction=/Game/Effects/Niagara/MasterMaterials/Functions/Common/FXCommon_NoiseControl}.NoiseValue
    UVControl_RGB=@n6
    UVControl_A=@n6
    "[Override]TimeInput"=@n7
   ) {materialFunction=&FXCommon_TextureSample}.RGB
   Fraction=ScalarParameter() {parameterName=f}
  )
  n9 = Multiply(A=@n8, B=ScalarParameter() {defaultValue=1, parameterName=aeta})
  n10 = VectorParameter() {parameterName=C, defaultValue=<rgba 22.766 7.7218 40 1>}
  Add(A=@n9, B=@n10.RGB)
  n11 = VectorParameter() {
   group=TipTailFade
   parameterName=TipTailFade_EdgeMask_InvertControl
   channelNames {a="", b=Power, g=Multiply, r=Offset}
   defaultValue=<rgba 0 15 1 1>
  }
  n12 = VectorParameter() {
   group=TipTailFade
   parameterName=TipTailFade_EdgeMask_Control
   channelNames {a="", b=Power, g=Multiply, r=Offset}
   defaultValue=<rgba 0 2 3 1>
  }
  n13 = MaterialFunctionCall!(
   UorV=MaterialFunctionCall(Float2=TextureCoordinate()) {materialFunction=&BreakOutFloat2Components}.G
   Invert_EdgeControl=@n11
   NoInvert_EdgeControl=@n12
  ) {materialFunction=/Game/Effects/Niagara/MasterMaterials/Functions/Common/FXCommon_EdgeMask_Ex}
  n14 = TextureObjectParameter() {
   group=CircleTexture
   parameterName=CircleTexture_Texture
   sortPriority=0
   texture=/Game/Accessories/FORT_Backpacks/Backpack_M_MED_OceanBreeze/FX/T_OceanBreeze_SplineB_t
  }
  n15 = VectorParameter() {group=CircleTexture, parameterName=CircleTexture_UVControl, channelNames {a=VPan, b=UPan, g=VScale, r=UScale}, defaultValue=<rgba 3 1 0.3 0>}
  n16 = MaterialFunctionCall(
   Texture=@n14
   UVControl_RGB=@n15
   UVControl_A=@n15
   "[Override]TimeInput"=@n7
   "[Override]UVInput"=MaterialFunctionCall(XY=TextureCoordinate()) {materialFunction=&Swizzle}.YX
  ) {materialFunction=&FXCommon_TextureSample}
  LinearInterpolate(B=ScalarParameter() {defaultValue=1.0444, parameterName=ref}, Alpha=Multiply(A=@n13.Mask, B=@n16.R)) {constA=1}
  MaterialFunctionCall(
   WidthBase=ScalarParameter() {defaultValue=7046.0806, parameterName=WidthBase}
   WidthTip=ScalarParameter() {defaultValue=19610.0723, parameterName=WidthTip}
  ) {materialFunction=&SplineThicken}
  MaterialFunctionCall(Emissive=Divide(A=Multiply(A=@n9, B=@n10.RGB)) {constB=2.2711}) {materialFunction=&MF_FX_EmissiveAdjustments}
  n17 = VectorParameter() {parameterName=Rainbow_Refine, defaultValue=<rgba 1 1 1 0.5>}
  MaterialFunctionCall(Texture=@n8, RGB=@n17.RGB, A=@n17.RGB) {materialFunction=&MF_FX_Refine}
  n18 = TextureCoordinate()
  n19 = MaterialFunctionCall(
   In=Add(
    A=MaterialFunctionCall(Float2=@n18) {materialFunction=&BreakOutFloat2Components}.G
    B=DynamicParameter() {paramNames [VOffset,Param2,Param3,Param4]}.VOffset
   )
   Contrast=Constant() {r=1}
  ) {materialFunction=&CheapContrast}
  n20 = VectorParameter() {parameterName=CircleTexture_Refine, defaultValue=<rgba 4 50 0.5 0.5>}
  n21 = MaterialFunctionCall(Texture=@n16.R, RGB=@n20.RGB, A=@n20.RGB) {materialFunction=&MF_FX_Refine}
  Multiply(A=Multiply(A=Saturate(None=@n19.Result), B=Multiply(A=@n13.Mask, B=@n21.Result)), B=ParticleColor().A)
 ]
}
