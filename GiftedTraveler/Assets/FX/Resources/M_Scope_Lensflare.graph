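# Additive, unlit Niagara sprite material that draws a lens flare, named for use on a scope.
# A grayscale flare mask is tinted by particle color, then passed through an LDR-to-HDR remap and an emissive adjustment function.
# The result and particle alpha feed a merge-to-lightweight-emitter prep function as color and opacity.
# An alternate spikeball flare texture parameter is present but unconnected.
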Material /GiftedTraveler/Assets/FX/Resources/M_Scope_Lensflare
graph {
 properties {bUsedWithNiagaraSprites=true, blendMode=BLEND_Additive, pixelDepthOffsetMode=PDOM_Legacy, shadingModel=MSM_Unlit, usedShadingModels=Unlit}
 nodes [
  n1 = ParticleColor()
  n2 = Multiply(
   A=TextureSampleParameter2D() {parameterName=FlareMask, samplerType=SAMPLERTYPE_Grayscale, texture=/Game/Effects/Fort_Effects/Textures/T_EFX_CameraFlare_01}.RGB
   B=@n1.RGB
  )
  n3 = MaterialFunctionCall(LDR Input=Clamp(None=@n2), HDR Tint=@n2) {materialFunction=&DeriveHDRfromLDR}
  MaterialFunctionCall!(
   Color=MaterialFunctionCall(Emissive=Divide(A=@n3.RemappedRange)) {materialFunction=&MF_FX_EmissiveAdjustments}.Result
   Opacity=@n1.A
   WPO
  ) {materialFunction=/Game/Effects/Fort_Effects/Materials/_Functions/MF_PrepareForMergeLWEmitter}
  TextureSampleParameter2D() {parameterName=TextureAlt, texture=/Game/Effects/Fort_Effects/Textures/Flares/T_Flare_Spikeball}
 ]
}
