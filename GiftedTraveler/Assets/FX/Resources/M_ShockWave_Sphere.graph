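# Additive, unlit, two-sided material for Niagara mesh particles, named as a shockwave sphere effect.
# A Fresnel term feeds a shared edge-mask function, shaped by inverted and non-inverted offset, multiply and power parameters.
# The mask is multiplied by particle alpha and passed through a parameterized depth fade to soften intersections.
# Particle color, scaled down by a constant, feeds an emissive-adjustment function.
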Material /GiftedTraveler/Assets/FX/Resources/M_ShockWave_Sphere
graph {
 properties {
  bAutomaticallySetUsageInEditor=false
  bUsedWithNiagaraMeshParticles=true
  blendMode=BLEND_Additive
  shadingModel=MSM_Unlit
  twoSided=true
  usedShadingModels=Unlit
 }
 nodes [
  n1 = ParticleColor()
  n2 = Fresnel() {baseReflectFraction=0, exponent=1}
  n3 = VectorParameter() {
   group=FresnelControl
   parameterName=FresnelControl_EdgeMask_InvertControl
   channelNames {a="", b=Power, g=Multiply, r=Offset}
   defaultValue=<rgba 0 1.8 4 1>
  }
  n4 = VectorParameter() {
   group=FresnelControl
   parameterName=FresnelControl_EdgeMask_Control
   channelNames {a="", b=Power, g=Multiply, r=Offset}
   defaultValue=<rgba 0 1.5 2 1>
  }
  DepthFade(
   Opacity=Multiply(
    A=@n1.A
    B=MaterialFunctionCall!(UorV=@n2, Invert_EdgeControl=@n3, NoInvert_EdgeControl=@n4) {materialFunction=/Game/Effects/Niagara/MasterMaterials/Functions/Common/FXCommon_EdgeMask_Ex}.Mask
   )
   FadeDistance=ScalarParameter() {defaultValue=15000, parameterName=DF}
  )
  MaterialFunctionCall(Emissive=Divide(A=@n1.RGB) {constB=2.2711}) {materialFunction=&MF_FX_EmissiveAdjustments}
 ]
}
