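# Unlit, translucent, two-sided Niagara effect material for a smoky spline or ribbon streak, thickened through SplineThicken plus noise-driven world position offset.
# A noise-distorted, axis-swapped smoke trail texture is desaturated, tinted, added to a refined detail texture and multiplied by particle color for emissive.
# Opacity combines the refined trail, tip/tail edge mask, particle alpha, squared camera distance fade and a Growth dynamic-parameter gradient that controls reveal along V.
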Material /GiftedTraveler/Assets/FX/Resources/M_Spline_Roll
graph {
 properties {
  bAutomaticallySetUsageInEditor=false
  bUsedWithNiagaraMeshParticles=true
  bUsedWithNiagaraRibbons=true
  bUsedWithNiagaraSprites=true
  blendMode=BLEND_Translucent
  shadingModel=MSM_Unlit
  twoSided=true
  usedShadingModels=Unlit
 }
 nodes [
  n1 = ScalarParameter() {defaultValue=16227.1396, parameterName=SplineThick}
  n2 = TextureObjectParameter() {group=WPONoise, parameterName=WPONoise_Texture, sortPriority=0, texture=/Game/Effects/Fort_Effects/Emotes/UltraLight/Texture/T_SoftNoise_BC7}
  n3 = VectorParameter() {group=WPONoise, parameterName=WPONoise_UVControl, channelNames {a=VPan, b=UPan, g=VScale, r=UScale}, defaultValue=<rgba 0 2 0 0.05>}
  Add(
   A=MaterialFunctionCall(WidthBase=@n1, WidthTip=@n1) {materialFunction=&SplineThicken}.WorldPositionOffset
   B=Multiply(
    A=ScalarParameter() {defaultValue=7303.0439, parameterName=WPOStrength}
    B=MaterialFunctionCall(Texture=@n2, UVControl_RGB=@n3, UVControl_A=@n3) {materialFunction=&FXCommon_TextureSample}.RGB
   )
  )
  n4 = TextureObjectParameter() {group=Detail, parameterName=Detail_Texture, sortPriority=0, texture=/Game/Items/Traps/Textures/T_Limbo_Ink_01_M}
  n5 = VectorParameter() {group=Detail, parameterName=Detail_UVControl, channelNames {a=VPan, b=UPan, g=VScale, r=UScale}, defaultValue=<rgba 1 0.5 0 0.1>}
  n6 = VectorParameter() {parameterName=Detail_Refine, defaultValue=<rgba 1 0 0 0.5>}
  n7 = Multiply(
   A=MaterialFunctionCall(
    Texture=MaterialFunctionCall(Texture=@n4, UVControl_RGB=@n5, UVControl_A=@n5) {materialFunction=&FXCommon_TextureSample}.G
    RGB=@n6.RGB
    A=@n6.RGB
   ) {materialFunction=&MF_FX_Refine}.Result
   B=VectorParameter() {parameterName=DetailPower, defaultValue=<rgba 0 0 0 1>}.RGB
  )
  n8 = TextureObjectParameter() {
   group=RibbonTexture
   parameterName=RibbonTexture_Texture
   sortPriority=0
   texture=/Game/Effects/Fort_Effects/Textures/Smoke/T_Smoke_Trail_Soft_2
  }
  n9 = TextureObjectParameter() {
   group=NoiseA
   parameterName=NoiseA_Texture
   samplerType=SAMPLERTYPE_LinearGrayscale
   sortPriority=0
   texture=/Game/Effects/Fort_Effects/Emotes/UltraLight/Texture/T_Cloud_Greyscale
  }
  n10 = VectorParameter() {group=NoiseA, parameterName=NoiseA_UVControl, channelNames {a=VPan, b=UPan, g=VScale, r=UScale}, defaultValue=<rgba 1 1 0 0>}
  n11 = MaterialFunctionCall(Texture=@n9, UVControl_RGB=@n10, UVControl_A=@n10) {materialFunction=&FXCommon_TextureSample}
  n12 = VectorParameter() {
   group=NoiseA
   parameterName=NoiseA_NoiseControl
   sortPriority=60
   channelNames {a=VStrength, b=UStrength, g=VOffset, r=UOffset}
   defaultValue=<rgba -0.5 -0.5 0 0>
  }
  n13 = VectorParameter() {
   group=RibbonTexture
   parameterName=RibbonTexture_UVControl
   channelNames {a=VPan, b=UPan, g=VScale, r=UScale}
   defaultValue=<rgba 0.8 1 0.03 0>
  }
  n14 = TextureCoordinate()
  n15 = Desaturation(
   None=MaterialFunctionCall(
    Texture=@n8
    AdditionalUVOffset=MaterialFunctionCall!(NoiseValue=@n11.RG, Control_RGB=@n12, Control_A=@n12) {materialFunction=/Game/Effects/Niagara/MasterMaterials/Functions/Common/FXCommon_NoiseControl}.NoiseValue
    UVControl_RGB=@n13
    UVControl_A=@n13
    "[Override]UVInput"=MaterialFunctionCall(XY=@n14) {materialFunction=&Swizzle}.YX
   ) {materialFunction=&FXCommon_TextureSample}.RGB
   Fraction=Constant() {r=1}
  )
  n16 = Multiply(A=Power(Base=@n15, Exp=ScalarParameter() {parameterName=DPower}), B=VectorParameter() {parameterName=Color, defaultValue=<rgba 0 0 0 1>}.RGB)
  n17 = ParticleColor()
  MaterialFunctionCall(Emissive=Multiply(A=Add(A=@n7, B=@n16), B=@n17.RGB)) {materialFunction=&MF_FX_EmissiveAdjustments}
  n18 = TextureCoordinate()
  n19 = VectorParameter() {parameterName=Growth_Grad, defaultValue=<rgba 3 2 0 1>}
  n20 = Multiply(
   A=Add(
    A=MaterialFunctionCall(Float2=@n18) {materialFunction=&BreakOutFloat2Components}.G
    B=DynamicParameter() {parameterIndex=3, paramNames [Growth,Param2,Param3,Param4]}.Growth
   )
   B=@n19.R
  )
  n21 = VectorParameter() {parameterName=RibbonTexture_Refine, defaultValue=<rgba 1 1.35 1 0.5>}
  n22 = TextureCoordinate()
  n23 = VectorParameter() {
   group=TipTailFade
   parameterName=TipTailFade_EdgeMask_InvertControl
   channelNames {a="", b=Power, g=Multiply, r=Offset}
   defaultValue=<rgba 0 20 1 1>
  }
  n24 = VectorParameter() {
   group=TipTailFade
   parameterName=TipTailFade_EdgeMask_Control
   channelNames {a="", b=Power, g=Multiply, r=Offset}
   defaultValue=<rgba 0 1.5 1.5 1>
  }
  n25 = Multiply(
   A=MaterialFunctionCall(Texture=@n15, RGB=@n21.RGB, A=@n21.RGB) {materialFunction=&MF_FX_Refine}.Result
   B=MaterialFunctionCall!(
    UorV=MaterialFunctionCall(Float2=@n22) {materialFunction=&BreakOutFloat2Components}.G
    Invert_EdgeControl=@n23
    NoInvert_EdgeControl=@n24
   ) {materialFunction=/Game/Effects/Niagara/MasterMaterials/Functions/Common/FXCommon_EdgeMask_Ex}.Mask
  )
  n26 = MaterialFunctionCall!(Distance=ScalarParameter() {defaultValue=40000, parameterName=FadeDistance}) {materialFunction=&/MF_CameraDistanceFade}
  Multiply(A=Power(Base=Saturate(None=@n20), Exp=@n19.G), B=Multiply(A=Multiply(A=@n25, B=@n17.A), B=Multiply(A=@n26.Result, B=@n26.Result)))
 ]
}
