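# M_TestFlare is an unlit Niagara sprite material that appears to be an unfinished flare effect test.
# It feeds a texture parameter and a UV scale and pan vector into the shared FXCommon_TextureSample function.
# A separate branch scales the camera-to-pixel world vector by -0.7, resembling a pull-toward-camera offset.
# No node connects to a material output, and a texture coordinate node sits unused.
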Material /GiftedTraveler/Assets/FX/Resources/M_TestFlare
graph {
 properties {bAutomaticallySetUsageInEditor=false, bUsedWithNiagaraSprites=true, shadingModel=MSM_Unlit, usedShadingModels=Unlit}
 nodes [
  n1 = VectorParameter() {group=S, parameterName=S_UVControl, channelNames {a=VPan, b=UPan, g=VScale, r=UScale}, defaultValue=<rgba 1 1 0 0>}
  MaterialFunctionCall(Texture=TextureObjectParameter() {group=S, parameterName=S_Texture, sortPriority=0}, UVControl_RGB=@n1, UVControl_A=@n1) {materialFunction=&FXCommon_TextureSample}
  Multiply(A=Subtract(A=WorldPosition() {worldPositionShaderOffset=WPT_ExcludeAllShaderOffsets}.XYZ, B=CameraPositionWS())) {constB=-0.7}
  TextureCoordinate()
 ]
}
