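# Translucent, unlit, two-sided Niagara ribbon FX material that draws a two-color streak scanning along UV V.
# Dynamic Time and Speed parameters, scaled by vertex color red, drive a saturated gradient that drops to zero at full, blending the colors and masking opacity.
# Opacity also combines two blended, refined textures (motor ribbon, soft smoke), particle alpha and tip, tail and side edge masks.
# A SplineThicken function takes one width for base and tip.
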Material /GiftedTraveler/Assets/FX/Resources/M_TopRibbon_OpVer
graph {
 properties {
  bAutomaticallySetUsageInEditor=false
  bUsedWithNiagaraMeshParticles=true
  bUsedWithNiagaraRibbons=true
  bUsedWithNiagaraSprites=true
  blendMode=BLEND_Translucent
  shadingModel=MSM_Unlit
  twoSided=true
  usedShadingModels=Unlit
 }
 nodes [
  n1 = OneMinus(None=MaterialFunctionCall(Float2=TextureCoordinate()) {materialFunction=&BreakOutFloat2Components}.G)
  n2 = VertexColor()
  n3 = Power(Base=@n2.R) {constExponent=1}
  Divide(A=@n1, B=@n3)
  n4 = ParticleColor()
  n5 = VectorParameter() {defaultValue=<rgba 1.1123 0.4 2 1>}
  n6 = VectorParameter() {parameterName=Param_1, defaultValue=<rgba 2 0.1 1.3491 1>}
  n7 = ScalarParameter() {defaultValue=20, parameterName=EndP}
  n8 = DynamicParameter() {paramNames [Time,Speed,Param3,Param4]}
  n9 = Multiply(A=@n8.Speed, B=@n3)
  n10 = Add(A=Subtract(A=@n7, B=Multiply(A=@n8.Time, B=@n9)), B=Multiply(A=@n1, B=Multiply(A=@n2.R) {constB=50}) {constB=17.599})
  n11 = ScalarParameter() {parameterName=StartP}
  n12 = Saturate(None=Divide(A=Subtract(A=@n10, B=@n11), B=@n7))
  n13 = Multiply(A=OneMinus(None=Floor(None=@n12)), B=@n12)
  MaterialFunctionCall(Emissive=Multiply(A=Divide(A=@n4.RGB) {constB=2.2711}, B=LinearInterpolate(A=@n5.RGB, B=@n6.RGB, Alpha=@n13))) {materialFunction=&MF_FX_EmissiveAdjustments}
  n14 = ScalarParameter() {defaultValue=10000, parameterName=SplineT}
  MaterialFunctionCall(WidthBase=@n14, WidthTip=@n14) {materialFunction=&SplineThicken}
  n15 = TextureObjectParameter() {
   group=MainTexture'
   parameterName=MainTexture'_Texture
   sortPriority=0
   texture=&/T_Backpack_MotorMonth_Ribbon
  }
  n16 = VectorParameter() {group=MainTexture', parameterName=MainTexture'_UVControl, channelNames {a=VPan, b=UPan, g=VScale, r=UScale}, defaultValue=<rgba 1 1 0 0>}
  n17 = MaterialFunctionCall(XY=TextureCoordinate()) {materialFunction=&Swizzle}
  n18 = VectorParameter() {parameterName=MainTexture'_Refine, defaultValue=<rgba 5 1 5 0.5>}
  n19 = MaterialFunctionCall(
   Texture=MaterialFunctionCall(Texture=@n15, UVControl_RGB=@n16, UVControl_A=@n16, "[Override]UVInput"=@n17.YX) {materialFunction=&FXCommon_TextureSample}.R
   RGB=@n18.RGB
   A=@n18.RGB
  ) {materialFunction=&MF_FX_Refine}
  n20 = TextureObjectParameter() {
   group=Tex2
   parameterName=Tex2_Texture
   samplerType=SAMPLERTYPE_Grayscale
   sortPriority=0
   texture=&T_Smoke_Trail_Soft
  }
  n21 = VectorParameter() {group=Tex2, parameterName=Tex2_UVControl, channelNames {a=VPan, b=UPan, g=VScale, r=UScale}, defaultValue=<rgba 1 1 0 0>}
  n22 = VectorParameter() {parameterName=Tex2_Refine, defaultValue=<rgba 1 1 2 0.5>}
  n23 = MaterialFunctionCall(
   Texture=MaterialFunctionCall(Texture=@n20, UVControl_RGB=@n21, UVControl_A=@n21, "[Override]UVInput"=@n17.YX) {materialFunction=&FXCommon_TextureSample}.R
   RGB=@n22.RGB
   A=@n22.RGB
  ) {materialFunction=&MF_FX_Refine}
  n24 = ScalarParameter() {defaultValue=0.5, parameterName=LerpTwoT}
  n25 = VectorParameter() {parameterName=Refine_Refine, defaultValue=<rgba 1 1 1 0.5>}
  n26 = Multiply(
   A=@n4.A
   B=MaterialFunctionCall(Texture=LinearInterpolate(A=@n19.Result, B=@n23.Result, Alpha=@n24), RGB=@n25.RGB, A=@n25.RGB) {materialFunction=&MF_FX_Refine}.Result
  )
  n27 = MaterialFunctionCall(Float2=TextureCoordinate() {coordinateIndex=1}) {materialFunction=&BreakOutFloat2Components}
  n28 = VectorParameter() {group=UVTFade, parameterName=UVTFade_EdgeMask_InvertControl, channelNames {a="", b=Power, g=Multiply, r=Offset}, defaultValue=<rgba 0 5 2 1>}
  n29 = VectorParameter() {group=UVTFade, parameterName=UVTFade_EdgeMask_Control, channelNames {a="", b=Power, g=Multiply, r=Offset}, defaultValue=<rgba 0 20 2 1>}
  n30 = MaterialFunctionCall!(UorV=@n27.G, Invert_EdgeControl=@n28, NoInvert_EdgeControl=@n29) {materialFunction=/Game/Effects/Niagara/MasterMaterials/Functions/Common/FXCommon_EdgeMask_Ex}
  n31 = Power(Base=@n13)
  n32 = VectorParameter() {
   group=ScanTipTail
   parameterName=ScanTipTail_EdgeMask_InvertControl
   channelNames {a="", b=Power, g=Multiply, r=Offset}
   defaultValue=<rgba 0 5 2 1>
  }
  n33 = VectorParameter() {
   group=ScanTipTail
   parameterName=ScanTipTail_EdgeMask_Control
   channelNames {a="", b=Power, g=Multiply, r=Offset}
   defaultValue=<rgba 0 20 2 1>
  }
  n34 = MaterialFunctionCall(UorV=@n13, Invert_EdgeControl=@n32, NoInvert_EdgeControl=@n33) {materialFunction=&FXCommon_EdgeMask_Ex}
  n35 = VectorParameter() {
   group=SideFade
   parameterName=SideFade_EdgeMask_InvertControl
   channelNames {a="", b=Power, g=Multiply, r=Offset}
   defaultValue=<rgba 1 1 1 1>
  }
  n36 = VectorParameter() {group=SideFade, parameterName=SideFade_EdgeMask_Control, channelNames {a="", b=Power, g=Multiply, r=Offset}, defaultValue=<rgba 1 1 1 1>}
  Multiply(
   A=Multiply(A=Multiply(A=@n26, B=@n30.Mask), B=Multiply(A=@n31, B=@n34.Mask))
   B=MaterialFunctionCall(UorV=@n27.R, Invert_EdgeControl=@n35, NoInvert_EdgeControl=@n36) {materialFunction=&FXCommon_EdgeMask_Ex}.Mask
  )
 ]
}
