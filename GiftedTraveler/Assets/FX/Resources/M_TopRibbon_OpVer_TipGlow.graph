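# Additive, unlit, two-sided Niagara material, apparently for ribbons, that draws a flickering glow tip travelling along the strip.
# Dynamic Time and Speed parameters, offset by vertex red, slide a clipped V gradient that remaps soft-flare UVs.
# A sine flickers the flare, blended with streaks, then tinted by particle color and a vertex-red hue shift.
# Edge masks on the gradient and second UV set compute fades, alongside a SplineThicken node.
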Material /GiftedTraveler/Assets/FX/Resources/M_TopRibbon_OpVer_TipGlow
graph {
 properties {
  bAutomaticallySetUsageInEditor=false
  bUsedWithNiagaraMeshParticles=true
  bUsedWithNiagaraRibbons=true
  bUsedWithNiagaraSprites=true
  blendMode=BLEND_Additive
  shadingModel=MSM_Unlit
  twoSided=true
  usedShadingModels=Unlit
 }
 nodes [
  n1 = OneMinus(None=MaterialFunctionCall(Float2=TextureCoordinate()) {materialFunction=&BreakOutFloat2Components}.G)
  n2 = VertexColor()
  n3 = Power(Base=@n2.R) {constExponent=1}
  Divide(A=@n1, B=@n3)
  n4 = ParticleColor()
  n5 = Divide(A=@n4.RGB) {constB=2.2711}
  n6 = VectorParameter() {group=TipT, parameterName=TipT_UVControl, channelNames {a=VPan, b=UPan, g=VScale, r=UScale}, defaultValue=<rgba 1 1 0 0>}
  n7 = TextureCoordinate()
  n8 = ScalarParameter() {defaultValue=-3, parameterName=FFF}
  n9 = ScalarParameter() {defaultValue=10, parameterName=EndP}
  n10 = DynamicParameter() {paramNames [Time,Speed,Param3,Param4]}
  n11 = Multiply(A=@n10.Speed, B=@n3)
  n12 = Add(A=Subtract(A=Add(A=@n8, B=@n9), B=Multiply(A=@n10.Time, B=@n11)), B=Multiply(A=@n1, B=Multiply(A=@n2.R) {constB=50}) {constB=17.599})
  n13 = ScalarParameter() {parameterName=StartP}
  n14 = Saturate(None=Divide(A=Subtract(A=@n12, B=@n13), B=@n9))
  n15 = Multiply(A=OneMinus(None=Floor(None=@n14)), B=@n14)
  n16 = MaterialFunctionCall(
   Texture=TextureObjectParameter() {group=TipT, parameterName=TipT_Texture, sortPriority=0, texture=/Game/Effects/Fort_Effects/Textures/Flares/T_Flare_Soft}
   UVControl_RGB=@n6
   UVControl_A=@n6
   "[Override]UVInput"=AppendVector(
    A=MaterialFunctionCall(Float2=@n7) {materialFunction=&BreakOutFloat2Components}.R
    B=@n15
   )
  ) {materialFunction=&FXCommon_TextureSample}
  n17 = Time() {bOverride_Period=true, period=6666}
  n18 = Multiply(A=@n2.R) {constB=100}
  n19 = Sine(None=Multiply(A=Add(A=@n17, B=@n18), B=ScalarParameter() {defaultValue=5, parameterName=SC5555}))
  n20 = Constant() {r=-1}
  n21 = Constant() {r=1}
  n22 = ScalarParameter() {defaultValue=0.5, parameterName=FlickL}
  n23 = LinearInterpolate(
   A=Multiply(
    A=@n16.R
    B=MaterialFunctionCall(Input=@n19, Input Low=@n20, Input High=@n21, Target Low=@n22, Target High=@n21) {materialFunction=&RemapValueRange}.Result
   )
   B=TextureSample(UVs=@n16.UV) {texture=/Game/Effects/Fort_Effects/Textures/Flares/T_Flare_Streaks}.RGB
  ) {constAlpha=0.25}
  n24 = VectorParameter() {parameterName=TipT_Refine, defaultValue=<rgba 3 1 0.5 0.5>}
  n25 = MaterialFunctionCall(Texture=@n23, RGB=@n24.RGB, A=@n24.RGB) {materialFunction=&MF_FX_Refine}
  n26 = Multiply(A=@n2.R, B=ScalarParameter() {defaultValue=0.5, parameterName=HS})
  n27 = VectorParameter() {parameterName=SC, defaultValue=<rgba 0.8157 5.419 7 1>}
  MaterialFunctionCall(Emissive=Multiply(
    A=Multiply(A=@n5, B=@n25.Result)
    B=MaterialFunctionCall(Hue Shift Percentage=@n26, Texture=@n27.RGB) {materialFunction=&HueShift}.Result
   )) {materialFunction=&MF_FX_EmissiveAdjustments}
  n28 = ScalarParameter() {defaultValue=35000, parameterName=SplineT}
  MaterialFunctionCall(WidthBase=@n28, WidthTip=@n28) {materialFunction=&SplineThicken}
  Multiply(A=@n4.A, B=@n23)
  n29 = VectorParameter() {
   group=ScanTipTail
   parameterName=ScanTipTail_EdgeMask_InvertControl
   channelNames {a="", b=Power, g=Multiply, r=Offset}
   defaultValue=<rgba 0 10 2 1>
  }
  n30 = VectorParameter() {
   group=ScanTipTail
   parameterName=ScanTipTail_EdgeMask_Control
   channelNames {a="", b=Power, g=Multiply, r=Offset}
   defaultValue=<rgba 0 10 2 1>
  }
  n31 = MaterialFunctionCall!(UorV=@n15, Invert_EdgeControl=@n29, NoInvert_EdgeControl=@n30) {materialFunction=/Game/Effects/Niagara/MasterMaterials/Functions/Common/FXCommon_EdgeMask_Ex}
  Multiply(A=Power(Base=@n15), B=@n31.Mask)
  n32 = MaterialFunctionCall(Float2=TextureCoordinate() {coordinateIndex=1}) {materialFunction=&BreakOutFloat2Components}
  n33 = VectorParameter() {group=UVTFade, parameterName=UVTFade_EdgeMask_InvertControl, channelNames {a="", b=Power, g=Multiply, r=Offset}, defaultValue=<rgba 0 5 2 1>}
  n34 = VectorParameter() {group=UVTFade, parameterName=UVTFade_EdgeMask_Control, channelNames {a="", b=Power, g=Multiply, r=Offset}, defaultValue=<rgba 0 3 2 1>}
  Multiply(
   A=Multiply(
    A=@n4.A
    B=MaterialFunctionCall(UorV=@n32.G, Invert_EdgeControl=@n33, NoInvert_EdgeControl=@n34) {materialFunction=&FXCommon_EdgeMask_Ex}.Mask
   )
   B=@n31.Mask
  )
 ]
}
