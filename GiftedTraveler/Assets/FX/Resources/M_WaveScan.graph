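# M_WaveScan is an additive, unlit Niagara material for a scan wave effect that projects a band around the particle onto scene geometry.
# With depth test off, it uses world position behind translucency relative to particle position to build angle and distance UVs.
# A noise texture perturbs the distance, which drives a plasma texture and an edge mask positioned by dynamic ScaleX and BandWidth.
# Refined texture and mask multiply with particle alpha; particle color feeds emissive.
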Material /GiftedTraveler/Assets/FX/Resources/M_WaveScan
graph {
 properties {
  bAutomaticallySetUsageInEditor=false
  bDisableDepthTest=true
  bUsedWithNiagaraMeshParticles=true
  bUsedWithNiagaraSprites=true
  blendMode=BLEND_Additive
  shadingModel=MSM_Unlit
  usedShadingModels=Unlit
 }
 nodes [
  n1 = ParticleColor()
  MaterialFunctionCall(Emissive=@n1.RGB) {materialFunction=&MF_FX_EmissiveAdjustments}
  n2 = TextureObjectParameter() {
   group=MainTexture
   parameterName=MainTexture_Texture
   samplerType=SAMPLERTYPE_LinearGrayscale
   sortPriority=0
   texture=/Game/Effects/Fort_Effects/Emotes/UltraLight/Texture/T_Plasma_A_B_Greyscale
  }
  n3 = VectorParameter() {group=MainTexture, parameterName=MainTexture_UVControl, channelNames {a=VPan, b=UPan, g=VScale, r=UScale}, defaultValue=<rgba 8 0.1 0 1>}
  n4 = MaterialFunctionCall() {materialFunction=&WorldPositionBehindTranslucency}
  n5 = ParticlePositionWS()
  n6 = Subtract(A=@n4.Result, B=@n5)
  n7 = MaterialFunctionCall(Float2=Normalize(VectorInput=ComponentMask(None=@n6) {g=true, r=true})) {materialFunction=&BreakOutFloat2Components}
  n8 = Multiply(A=Arctangent2Fast(Y=@n7.R, X=@n7.R)) {constB=2}
  n9 = MaterialFunctionCall(Multiplier=Constant() {r=2}) {materialFunction=&Pi}
  n10 = Divide(A=Add(A=Divide(A=@n8, B=@n9.Result)))
  n11 = TextureObjectParameter() {
   group=Noise
   parameterName=Noise_Texture
   samplerType=SAMPLERTYPE_LinearColor
   sortPriority=0
   texture=/Game/Items/Traps/Textures/T_Plasma_Flames_02_M
  }
  n12 = VectorParameter() {group=Noise, parameterName=Noise_UVControl, channelNames {a=VPan, b=UPan, g=VScale, r=UScale}, defaultValue=<rgba 5 0.01 0 -1>}
  n13 = Distance(A=@n4.Result, B=@n5)
  n14 = MaterialFunctionCall(Texture=@n11, UVControl_RGB=@n12, UVControl_A=@n12, "[Override]UVInput"=AppendVector(A=@n10, B=Multiply(A=@n13) {constB=0.0001})) {materialFunction=&FXCommon_TextureSample}
  n15 = VectorParameter() {parameterName=Noise_Refine, defaultValue=<rgba 2.2331 1 0 0.5>}
  n16 = VectorParameter() {
   group=Noise
   parameterName=Noise_NoiseControl
   sortPriority=60
   channelNames {a=VStrength, b=UStrength, g=VOffset, r=UOffset}
   defaultValue=<rgba 0 -0.5 46271.5234 0>
  }
  n17 = Add(
   A=MaterialFunctionCall!(
    NoiseValue=MaterialFunctionCall(Texture=@n14.R, RGB=@n15.RGB, A=@n15.RGB) {materialFunction=&MF_FX_Refine}.Result
    Control_RGB=@n16
    Control_A=@n16
   ) {materialFunction=/Game/Effects/Niagara/MasterMaterials/Functions/Common/FXCommon_NoiseControl}.NoiseValue
   B=@n13
  )
  n18 = MaterialFunctionCall(Texture=@n2, UVControl_RGB=@n3, UVControl_A=@n3, "[Override]UVInput"=AppendVector(A=@n10, B=Multiply(A=@n17) {constB=0.0001})) {materialFunction=&FXCommon_TextureSample}
  n19 = VectorParameter() {parameterName=MainTexture_Refine, defaultValue=<rgba 0.5 1 0.5 0.5>}
  n20 = DynamicParameter() {paramNames [ScaleX,BandWidth,Param3,Param4]}
  n21 = Multiply(A=@n20.ScaleX) {constB=50}
  n22 = Divide(A=Subtract(A=@n17, B=Subtract(A=@n21, B=@n20.BandWidth)), B=@n20.BandWidth)
  n23 = MaterialFunctionCall!(TexCoord=@n22, ClampChannel=Constant2Vector() {g=1, r=1}) {materialFunction=/Game/Effects/Fort_Effects/MaterialFunctions/MF_FX_TilingMethod}
  n24 = VectorParameter() {
   group=ScanEdgeControl
   parameterName=ScanEdgeControl_EdgeMask_InvertControl
   channelNames {a="", b=Power, g=Multiply, r=Offset}
   defaultValue=<rgba 0 6 2 1>
  }
  n25 = VectorParameter() {
   group=ScanEdgeControl
   parameterName=ScanEdgeControl_EdgeMask_Control
   channelNames {a="", b=Power, g=Multiply, r=Offset}
   defaultValue=<rgba 0 1 1.5 1>
  }
  n26 = MaterialFunctionCall!(UorV=Saturate(None=Multiply(A=@n23.Result, B=@n22)), Invert_EdgeControl=@n24, NoInvert_EdgeControl=@n25) {materialFunction=/Game/Effects/Niagara/MasterMaterials/Functions/Common/FXCommon_EdgeMask_Ex}
  n27 = VectorParameter() {parameterName=ScanEdgeControl_Refine, defaultValue=<rgba 3 3 0.728 0.5>}
  Multiply(
   A=@n1.A
   B=Multiply(
    A=MaterialFunctionCall(Texture=@n18.R, RGB=@n19.RGB, A=@n19.RGB) {materialFunction=&MF_FX_Refine}.Result
    B=MaterialFunctionCall(Texture=@n26.Mask, RGB=@n27.RGB, A=@n27.RGB) {materialFunction=&MF_FX_Refine}.Result
   )
  )
 ]
}
